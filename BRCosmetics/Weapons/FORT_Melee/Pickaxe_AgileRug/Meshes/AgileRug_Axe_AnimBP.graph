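# Animation Blueprint for the AgileRug pickaxe cosmetic that rotates five base bones of the axe mesh while a melee animation plays.
# The input pose is cached, then run through a chain of additive bone-space Modify Bone nodes that apply the same pitch offset to each bone.
# A Blend Poses by bool on bIsPlayingMeleeAnim eases between this modified pose and the untouched cached pose.
# The three declared booleans are unused in the graphs shown.
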AnimBlueprint /BRCosmetics/Weapons/FORT_Melee/Pickaxe_AgileRug/Meshes/AgileRug_Axe_AnimBP
target_skeleton=&/AgileRug_Axe_Skeleton
graphs [
 AnimGraph {nodes [
   n1 = "Transform (Modify) Bone - Bone: base_top_c"(ComponentPose=Local To Component(LocalPose=Use cached pose 'CR_Start'()), Alpha=1, Translation="0,0,0", Rotation="26,0,0", Scale="1,1,1") {
    rotationMode=BMM_Additive
    rotationSpace=BCS_BoneSpace
    translationMode=BMM_Additive
    translationSpace=BCS_BoneSpace
    alphaBoolBlend {blendOption=Linear}
    boneToModify {boneName=base_top_c}
    rotation {pitch=26}
   }
   n2 = "Transform (Modify) Bone - Bone: base_side_l"(ComponentPose=@n1, Alpha=1, Translation="0,0,0", Rotation="26,0,0", Scale="1,1,1") {
    rotationMode=BMM_Additive
    rotationSpace=BCS_BoneSpace
    translationMode=BMM_Additive
    translationSpace=BCS_BoneSpace
    alphaBoolBlend {blendOption=Linear}
    boneToModify {boneName=base_side_l}
    rotation {pitch=26}
   }
   n3 = "Transform (Modify) Bone - Bone: base_low_l"(ComponentPose=@n2, Alpha=1, Translation="0,0,0", Rotation="26,0,0", Scale="1,1,1") {
    rotationMode=BMM_Additive
    rotationSpace=BCS_BoneSpace
    translationMode=BMM_Additive
    translationSpace=BCS_BoneSpace
    alphaBoolBlend {blendOption=Linear}
    boneToModify {boneName=base_low_l}
    rotation {pitch=26}
   }
   n4 = "Transform (Modify) Bone - Bone: base_low_r"(ComponentPose=@n3, Alpha=1, Translation="0,0,0", Rotation="26,0,0", Scale="1,1,1") {
    rotationMode=BMM_Additive
    rotationSpace=BCS_BoneSpace
    translationMode=BMM_Additive
    translationSpace=BCS_BoneSpace
    alphaBoolBlend {blendOption=Linear}
    boneToModify {boneName=base_low_r}
    rotation {pitch=26}
   }
   n5 = "Transform (Modify) Bone - Bone: base_side_r"(ComponentPose=@n4, Alpha=1, Translation="0,0,0", Rotation="26,0,0", Scale="1,1,1") {
    rotationMode=BMM_Additive
    rotationSpace=BCS_BoneSpace
    translationMode=BMM_Additive
    translationSpace=BCS_BoneSpace
    alphaBoolBlend {blendOption=Linear}
    boneToModify {boneName=base_side_r}
    rotation {pitch=26}
   }
   Output Pose(Result=Blend Poses by bool(
     BlendPose_0=Component To Local(ComponentPose=@n5)
     BlendPose_1=Use cached pose 'CR_Start'()
     BlendTime_1=0.15
     bActiveValue=Get bIsPlayingMeleeAnim()
    ) {blendType=Linear}) {layerGroup=None}
   AnimGraphNode_SubInput_0 = Input Pose()
   Save cached pose 'CR_Start'(Pose=@AnimGraphNode_SubInput_0) {cacheName=CR_Start}
  ]}
 EventGraph {nodes [TryGetPawnOwner()]}
]
variables [
 BeginPASwing {type=boolean}
 NewVar {type=boolean}
 PickaxeSwing {type=boolean}
]
